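# Niagara idle effect for the DuneBriefMerry pickaxe, combining an equip flourish with a persistent fire look.
# Equip emits a short burst of red mesh particles that swirl inward and shrink, while EquipTracer fades a bright mesh through animated material parameters.
# Idle_Fire keeps three partitioned particles attached along skeletal mesh bones through a list-curve sampler, feeding layered mesh and ribbon renderers with per-partition color and width plus waveform-driven pulsing.
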NiagaraSystem /BRCosmetics/Weapons/FORT_Melee/Pickaxe_DuneBriefMerry/FX/NS_Pickaxe_DuneBriefMerry_Idle
dependencies {
 usedDataInterfaces [
  &NiagaraDataInterfaceCurve
  &NiagaraDataInterfaceArrayColor
  &NiagaraDataInterfaceArrayFloat
  &NiagaraDataInterfaceArrayInt32
  &NiagaraDataInterfaceCamera
  &NiagaraDataInterfaceSkeletalMesh
 ]
 usedDynamicInputs []
 usedModules [
  &self:SystemSpawnScript.SystemScriptSource.SystemScriptGraph.NiagaraNodeAssignment_0.SetVariables_#1
  &SystemState
  &self:SystemSpawnScript.SystemScriptSource.SystemScriptGraph.NiagaraNodeAssignment_1.SetVariables_#2
 ]
}
emitters {
 Equip {
  emitterUpdateScript [
   &EmitterState(Inactive Response=::NewEnumerator1, Life Cycle Mode=::NewEnumerator1, Loop Behavior=::NewEnumerator1, Loop Duration=0.65, UseLoopDelay=false)
   &SpawnBurst_Instantaneous(Spawn Count=30, Spawn Time=0.25, Use Spawn Probability=false, UseLoopCountLimit=false)
   &SpawnRate(
    Spawn Probability=&MakeCustomFloatFromBool(
     Bool=&SetBoolByFloatComparison(A=$Emitter.NormalizedLoopAge, B=0.5, Comparison Type=::NewEnumerator6)
     True Float=1
     False Float=0
    )
    SpawnRate=200
    Use Spawn Probability=true
   )
  ]
  particleSpawnScript [
   &InitializeParticle(
    Color=&RandomRangeLinearColor(Minimum=<rgba 0.2396 0.0071 0.0117 1>, Maximum=<rgba 0.4635 0.0232 0.0276 1>)
    Lifetime=&RandomRangeFloat(Minimum=0.25, Maximum=0.5)
    Mesh Scale Mode=::NewEnumerator3
    Mesh Uniform Scale=&RandomRangeFloat(Minimum=0.2, Maximum=0.75)
    Sprite Size Mode=::NewEnumerator3
    Uniform Sprite Size=&RandomRangeFloat(Minimum=10, Maximum=20)
    UsePositionOffset=false
    Write Material Random=true
   )
   /SetVariables(Particles.ParticleRandom:NiagaraFloat=&RandomRangeFloat(Minimum=0, Maximum=1)) {isSetParameters=true}
   /Game/Effects/Niagara/Modules/Curves/FN_ListCurve(
    Apply Position=true
    Curve Index:NiagaraFloat=&FloatFromCurve(
     FloatCurve=DataInterface() {propertyValues {Curve {keys [
         {arriveTangent=0.5817, interpMode=RCIM_Cubic, leaveTangent=0.5817, leaveTangentWeight=0.1157, tangentMode=RCTM_User, time=0, value=1}
         {tangentMode=RCTM_User, time=0.3, value=0}
         <at 0.35 0.121>
        ]}}}
     CurveIndex=$Emitter.LoopedAge
    )
    Input Type=2
    Number Inputs Switch=0
    Position Coordinate Space=1
    Vector_01:Vector3f=&RandomRangeVector(Minimum=<0 0 20>, Maximum=<0 0 28>)
    Vector_02:Vector3f=<0 0 -10>
   )
  ]
  particleUpdateScript [
   &SolveForcesAndVelocity(Clamp Velocity=false, Limit Acceleration=false)
   &ParticleState()
   /SetVariables(
    Particles.Position=&self:ScratchDynamicInput(
     Theta:NiagaraFloat=$Particles.ParticleRandom
     Evaluation Type:NiagaraInt32=0
     Offset:Vector3f=Linked() {linkedVariable {name=Particles.Initial.Position, type=NiagaraPosition}}
     Rotation:Vector3f=<0 0 0>
     Radius:NiagaraFloat=&FloatFromCurve(
      FloatCurve=DataInterface() {propertyValues {Curve {keys [
          {arriveTangent=6.8794, interpMode=RCIM_Cubic, leaveTangent=6.8794, tangentMode=RCTM_User, time=0, value=1}
          {arriveTangent=0.8488, interpMode=RCIM_Cubic, leaveTangent=0.8488, tangentMode=RCTM_User, time=0.15, value=1.5}
          {arriveTangent=-1.1354, interpMode=RCIM_Cubic, leaveTangent=-1.1354, tangentMode=RCTM_User, time=0.3, value=1.5}
          {interpMode=RCIM_Cubic, tangentMode=RCTM_User, time=1, value=0.25}
         ]}}}
      Scale Curve=&FloatFromCurve(
       FloatCurve=DataInterface() {propertyValues {Curve {keys [{interpMode=RCIM_Cubic, tangentMode=RCTM_User, time=0, value=1}, {arriveTangent=-3.9037, tangentMode=RCTM_User, time=0.5, value=0}]}}}
       CurveIndex=$Emitter.LoopedAge
       Scale Curve=10
      )
     )
     Random Seed:NiagaraInt32=0
     U Coverage:NiagaraFloat=0
    )
    Particles.SpriteSize=&Multiply_Vector2D_ByFloat(
     Float=&FloatFromCurve(FloatCurve=DataInterface() {propertyValues {Curve {keys [
          {interpMode=RCIM_Cubic, tangentMode=RCTM_User, time=0, value=0}
          {interpMode=RCIM_Cubic, tangentMode=RCTM_User, time=0.25, value=1}
          {arriveTangent=-4.7366, interpMode=RCIM_Cubic, leaveTangent=-4.7366, tangentMode=RCTM_User, time=1, value=0}
         ]}}})
     Vector2D=$Particles.Initial.SpriteSize
    )
   ) {isSetParameters=true}
   &CameraOffset()
   &ScaleMeshSize(Scale Factor=&Multiply_VectorByFloat(
     Vector=<1 1 1>
     Float=&FloatFromCurve(FloatCurve=DataInterface() {propertyValues {Curve {keys [
          {arriveTangent=2.9773, interpMode=RCIM_Cubic, leaveTangent=2.9773, tangentMode=RCTM_User, time=0, value=1}
          {tangentMode=RCTM_User, time=1, value=0.25}
         ]}}})
    ))
   &ScaleColor(
    Scale RGB=&Multiply_VectorByFloat(
     Vector=<1 1 1>
     Float=&FloatFromCurve(FloatCurve=DataInterface() {propertyValues {Curve {keys [{arriveTangent=-29.5871, interpMode=RCIM_Cubic, leaveTangent=-29.5871, tangentMode=RCTM_User, time=0, value=10}, <at 0.5 1>]}}})
    )
    ScaleA=true
    ScaleRGB=true
   )
  ]
  rendererClasses [&NiagaraMeshRendererProperties]
  renderers [
   {rendererClass=&NiagaraMeshRendererProperties, rendererIndex=0}
   {rendererClass=&NiagaraMeshRendererProperties, rendererIndex=1}
  ]
 }
 EquipTracer {
  emitterUpdateScript [
   &EmitterState(Inactive Response=::NewEnumerator1, Life Cycle Mode=::NewEnumerator1, Loop Behavior=::NewEnumerator1, Loop Duration=0.85, UseLoopDelay=false)
   &SpawnBurst_Instantaneous(Use Spawn Probability=false, UseLoopCountLimit=false)
  ]
  particleSpawnScript [&InitializeParticle(
    Color=<rgba 25 25 25 1>
    Lifetime=0.65
    Mesh Scale Mode=::NewEnumerator3
    Position Offset=<0 0 5>
    UsePositionOffset=true
    Write Material Random=true
   )]
  particleUpdateScript [
   &ParticleState()
   &DynamicMaterialParameters(
    Index 0 Param 1=1
    Index 0 Param 2=1
    Index 0 Param 4=1
    Index 1 Param 3=&FloatFromCurve(FloatCurve=DataInterface() {propertyValues {Curve {keys [
         {arriveTangent=-0.2519, interpMode=RCIM_Cubic, leaveTangent=-0.2519, tangentMode=RCTM_User, time=0, value=1}
         {arriveTangent=-1.804, tangentMode=RCTM_User, time=1, value=-1}
        ]}}})
    Module.Index0Float0Write=true
    Module.Index0Float1Write=true
    Module.Index0Float2Write=true
    Module.Index0Float3Write=true
    Module.Index1Float0Write=true
    Module.Index1Float1Write=true
    Module.Index1Float2Write=true
    Module.Index1Float3Write=true
    Write Parameter Index 1=::NewEnumerator1
   )
   &ScaleColor(
    Scale Alpha=&FloatFromCurve(
     FloatCurve=DataInterface() {propertyValues {Curve {keys [
         {interpMode=RCIM_Cubic, tangentMode=RCTM_User, time=0, value=1}
         {arriveTangent=-1, interpMode=RCIM_Cubic, leaveTangent=-1, time=0.4847, value=0.8588}
         {arriveTangent=-0.6755, tangentMode=RCTM_User, time=1, value=0}
        ]}}}
     CurveIndex=$Emitter.NormalizedLoopAge
    )
    ScaleA=true
    ScaleRGB=true
   )
   &CameraOffset(Camera Offset Amount=4)
  ]
  rendererClasses [&NiagaraMeshRendererProperties]
  renderers [{rendererClass=&NiagaraMeshRendererProperties, rendererIndex=0}]
 }
 Idle_Fire {
  emitterUpdateScript [
   &EmitterState(
    Inactive Response=::NewEnumerator1
    Life Cycle Mode=::NewEnumerator1
    Loop Behavior=::NewEnumerator1
    Loop Delay=0.1
    Loop Duration Mode=::NewEnumerator1
    UseLoopDelay=true
   )
   &SpawnBurst_Instantaneous(Spawn Count=3, Use Spawn Probability=false, UseLoopCountLimit=false)
  ]
  particleSpawnScript [
   &PartitionParticles(Normalized Index Includes One=true, Number Of Partitions=3)
   &InitializeParticle(
    Color=&SelectLinearColorFromArray(
     Array Sampling Mode=::NewEnumerator1
     Direct Array Index=$Particles.PartitionParticles.Partition
     Color Selection Array=DataInterface() {propertyValues {ColorData [<rgba 3 3 3 1>, <rgba 0.2 0.2 0.2 1>, <rgba 0 0 0 1>]}}
     Random Seed=?
    )
    Mesh Scale=<0.8 0.8 0.9>
    Mesh Scale Mode=::NewEnumerator4
    Position Offset=<0 0 18>
    Ribbon Width=&SelectFloatFromArray(Direct Array Index=$Particles.PartitionParticles.Partition, Float Selection Array=DataInterface() {propertyValues {FloatData [10,20,30]}})
    Ribbon Width Mode=::NewEnumerator1
    Sprite Size Mode=::NewEnumerator3
    UsePositionOffset=false
    Write Material Random=true
   )
   /SetVariables(Particles.VisCore:NiagaraInt32=&SelectIntFromArray(Direct Array Index=$Particles.PartitionParticles.Partition, Int Selection Array=DataInterface() {propertyValues {IntData [1,0,0]}})) {isSetParameters=true}
  ]
  particleUpdateScript [
   &ParticleState(Kill Particles When Lifetime Has Elapsed=false)
   &DynamicMaterialParameters(
    Index 0 Param 1=1
    Index 0 Param 2=1
    Index 0 Param 3=&FloatFromCurve(
     FloatCurve=DataInterface() {propertyValues {Curve {keys [<at 0.65 0>,<at 1 1>]}}}
     CurveIndex=$Particles.Initial.PartitionParticles.NormalizedPartition
    )
    Index 0 Param 4=1
    Index 2 Param 1=&Waveform("Waveform [1]"=::NewEnumerator0, "[1] Frequency"=1, "[1] Amplitude Min/Max"=<0 0.05>)
    Index 2 Param 2=&Waveform("Waveform [1]"=::NewEnumerator7, "[1] Frequency"=3, "[1] Amplitude Min/Max"=<0 0.05>)
    Module.Index0Float0Write=true
    Module.Index0Float1Write=true
    Module.Index0Float2Write=true
    Module.Index0Float3Write=true
    Module.Index2Float0Write=true
    Module.Index2Float1Write=true
    Module.Index2Float2Write=true
    Module.Index2Float3Write=true
    Write Parameter Index 2=::NewEnumerator1
   )
   /Game/Effects/Niagara/Modules/Curves/FN_ListCurve_SkeletalMeshSampler(
    Apply Orientation=true
    Apply Position=true
    Bone Coordinate Space=2
    Bone Position Contribution:NiagaraFloat=&FloatFromCurve(
     FloatCurve=DataInterface() {propertyValues {Curve {keys [<at 0.5 1>,<at 1 0.5>]}}}
     CurveIndex=$Particles.Initial.PartitionParticles.NormalizedPartition
    )
    Curve Index:NiagaraFloat=&Multiply_Float(
     A=&SelectFloatFromArray(Direct Array Index=$Particles.PartitionParticles.Partition, Float Selection Array=DataInterface() {propertyValues {FloatData [0,0.5,1]}})
     B=&FloatFromCurve(FloatCurve=DataInterface() {propertyValues {Curve {keys [<at 0 0.2>,<at 0.2 1>]}}}, CurveIndex=$Particles.Age)
    )
    Index_02=1
    Index_03=1
    Number of Inputs=1
    Offset Coordinate Space=3
    Offset_01:Vector3f=<0 0 -1>
    Offset_03:Vector3f=<2 0 5>
    Selection Type=0
    Skeletal Mesh:NiagaraDataInterfaceSkeletalMesh=$System.SkeletalMesh
   )
   /SetVariables(
    Particles.CoreCameraOffset=-1
    Particles.Fire Blade Scale:Vector3f=<1 1 1>
    Particles.FireCuffScale:Vector3f=&Lerp_Vector(
     A=<0.8 0.8 1>
     B=<0.8 0.8 1.25>
     Alpha=&Waveform("Waveform [1]"=::NewEnumerator0, "[1] Frequency"=0.5, "[1] Amplitude Min/Max"=<0 1>)
    )
    Particles.ribbon Core Camera=2
    Particles.ribbon2:NiagaraPosition=&AddVectorToPosition(
     Position=&AddVectorToPosition(Position=$Particles.Position, Vector=<0 0 0>)
     Vector=&Multiply_VectorByFloat(Vector=&ReturnCameraVector(Camera Vectors=::NewEnumerator1, Coordinate Space=::Simulation), Float=-2)
    )
    Particles.ribbon2 width:NiagaraFloat=&SelectFloatFromArray(Direct Array Index=$Particles.PartitionParticles.Partition, Float Selection Array=DataInterface() {propertyValues {FloatData [1,15,0]}})
   ) {isSetParameters=true}
   &CameraOffset(Camera Offset Amount=1)
  ]
  rendererClasses [&NiagaraMeshRendererProperties, &NiagaraRibbonRendererProperties]
  renderers [
   {rendererClass=&NiagaraMeshRendererProperties, rendererIndex=0}
   {rendererClass=&NiagaraRibbonRendererProperties, rendererIndex=1}
   {rendererClass=&NiagaraMeshRendererProperties, rendererIndex=2}
   {rendererClass=&NiagaraMeshRendererProperties, rendererIndex=3}
   {rendererClass=&NiagaraRibbonRendererProperties, rendererIndex=4}
  ]
 }
}
summary {
 systemName=NS_Pickaxe_DuneBriefMerry_Idle
 emitters [
  {emitterName=Equip, rendererClasses [&NiagaraMeshRendererProperties]}
  {emitterName=EquipTracer, rendererClasses [&NiagaraMeshRendererProperties]}
  {emitterName=Idle_Fire, rendererClasses [&NiagaraMeshRendererProperties, &NiagaraRibbonRendererProperties]}
 ]
 userVariables [
  User.bIsFrontEnd {type=NiagaraBool}
  User.bIsFrontEndPreview {type=NiagaraBool}
  User.bIsSwinging {type=NiagaraBool}
  User.VariantColor {type=NiagaraFloat, defaultValue=0, description=""}
  User.VariantToon {type=NiagaraFloat, defaultValue=1, description=""}
 ]
}
