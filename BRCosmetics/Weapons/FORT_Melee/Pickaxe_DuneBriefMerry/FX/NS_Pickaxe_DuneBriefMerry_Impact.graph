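# Niagara impact effect for the DuneBriefMerry pickaxe, throwing short-lived bursts along a system direction vector set for pickaxe swings.
# Directional_Elements packs five spawn groups into one emitter, using group-indexed switches for motion and look, with VisibilityTag separating them across five sprite renderers.
# Spark-like group counts vary by platform, while front-end scale and time of day adjust size, speed and color.
# Minimal adds a single-particle sprite/mesh burst; Fire_Sprites is disabled.
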NiagaraSystem /BRCosmetics/Weapons/FORT_Melee/Pickaxe_DuneBriefMerry/FX/NS_Pickaxe_DuneBriefMerry_Impact
dependencies {
 usedDataInterfaces [
  &NiagaraDataInterfacePlatformSet
  &NiagaraDataInterfaceColorCurve
  &NiagaraDataInterfaceSpriteRendererInfo
  &NiagaraDataInterfaceCurve
  &NiagaraDataInterfaceVector2DCurve
  &NiagaraDataInterfaceArrayFloat
  &NiagaraDataInterfaceArrayInt32
  &NiagaraDataInterfaceArrayFloat2
  &NiagaraDataInterfaceArrayFloat3
 ]
 usedDynamicInputs [/Game/Effects/Niagara/Modules/Utility/FN_DirectionVectorByPickaxeType]
 usedModules [
  &self:SystemSpawnScript.SystemScriptSource.SystemScriptGraph.NiagaraNodeAssignment_1.SetVariables_#1
  &self:SystemSpawnScript.SystemScriptSource.SystemScriptGraph.NiagaraNodeAssignment_0.SetVariables_#2
  &SystemState
 ]
}
emitters {
 Directional_Elements {
  emitterUpdateScript [
   &EmitterState()
   &SpawnBurst_Instantaneous(Spawn Count=1, Spawn Group=4, Use Spawn Probability=false) {module=SpawnBurst_Instantaneous004}
   &SpawnBurst_Instantaneous(
    Spawn Count=&MakeIntFromFloat(Float=&FN_PickFloatDependOnPlatform(
      TargetPlatform=DataInterface() {propertyValues {Platforms {qualityLevelMask=-29}}}
      TargetPlatformActive=&MakeFloatFromInt(Int=&RandomRangeInt(Minimum=4, Maximum=6))
      TargetPlatformNoActive=&MakeFloatFromInt(Int=&RandomRangeInt(Minimum=12, Maximum=16))
     ))
    Spawn Group=2
    Use Spawn Probability=false
    UseLoopCountLimit=false
   ) {module=SpawnBurst_Instantaneous003}
   &SpawnBurst_Instantaneous(
    Spawn Count=&MakeIntFromFloat(Float=&FN_PickFloatDependOnPlatform(
      TargetPlatform=DataInterface() {propertyValues {Platforms {qualityLevelMask=-29}}}
      TargetPlatformActive=&MakeFloatFromInt(Int=&RandomRangeInt(Minimum=3, Maximum=5))
      TargetPlatformNoActive=&MakeFloatFromInt(Int=&RandomRangeInt(Minimum=8, Maximum=10))
     ))
    Spawn Group=3
    Use Spawn Probability=false
    UseLoopCountLimit=false
   ) {module=SpawnBurst_Instantaneous002}
   &SpawnBurst_Instantaneous(Spawn Count=2, Spawn Group=1, Use Spawn Probability=false, UseLoopCountLimit=false) {module=SpawnBurst_Instantaneous001}
   &SpawnBurst_Instantaneous(Use Spawn Probability=false, UseLoopCountLimit=false)
  ]
  particleSpawnScript [
   &InitializeParticle(
    Color=&MergeHelper_ColorSwitch(
     0=<rgba 2 1.1957 0.1417 1>
     1=&RandomRangeLinearColor(Minimum=<rgba 0 0 0 0.7>, Maximum=<rgba 0 0 0 1>)
     2=&ColorFromCurve(
      ColorCurve=DataInterface() {propertyValues {
        AlphaCurve {keys [<at 0 1 RCIM_Cubic>]}
        BlueCurve {keys [
          <at 0.2775 0.132>
          <at 0.3074 2>
          {arriveTangent=-4.3243, interpMode=RCIM_Cubic, leaveTangent=-4.3243, time=0.7173, value=1}
          <at 0.7357 0.1478 RCIM_Cubic>
         ]}
        GreenCurve {keys [
          <at 0.2775 1.088>
          <at 0.3074 2>
          {arriveTangent=-3.2432, interpMode=RCIM_Cubic, leaveTangent=-3.2432, time=0.7173, value=1}
          <at 0.7357 0.6109 RCIM_Cubic>
         ]}
        RedCurve {keys [<at 0.2775 3>,<at 0.3074 2>,<at 0.7173 1 RCIM_Cubic>,<at 0.7357 2 RCIM_Cubic>]}
       }}
      CurveIndex=&RandomRangeFloat(Minimum=0, Maximum=1)
     )
     3=&ColorFromCurve(
      ColorCurve=DataInterface() {propertyValues {
        AlphaCurve {keys [<at 0 1 RCIM_Cubic>]}
        BlueCurve {keys [
          <at 0.2775 0.132>
          <at 0.3074 2>
          {arriveTangent=-4.3243, interpMode=RCIM_Cubic, leaveTangent=-4.3243, time=0.7173, value=1}
          <at 0.7357 0.1478 RCIM_Cubic>
         ]}
        GreenCurve {keys [
          <at 0.2775 1.088>
          <at 0.3074 2>
          {arriveTangent=-3.2432, interpMode=RCIM_Cubic, leaveTangent=-3.2432, time=0.7173, value=1}
          <at 0.7357 0.6109 RCIM_Cubic>
         ]}
        RedCurve {keys [<at 0.2775 3>,<at 0.3074 2>,<at 0.7173 1 RCIM_Cubic>,<at 0.7357 2 RCIM_Cubic>]}
       }}
      CurveIndex=&RandomRangeFloat(Minimum=0, Maximum=1)
     )
     4=<rgba 1 1 1 0.25>
     Index=$Emitter.SpawnGroup
     Num=::NewEnumerator3
    )
    Lifetime=&MergeHelper_FloatSwitch(
     0=0.16
     1=&RandomRangeFloat(Minimum=0.1, Maximum=0.15)
     2=&RandomRangeFloat(Minimum=0.05, Maximum=0.45)
     3=&RandomRangeFloat(Minimum=0.24, Maximum=0.47)
     4=0.32
     Index=$Emitter.SpawnGroup
     Num=::NewEnumerator3
    )
    Sprite Size=&MergeHelper_Vector2Switch(
     0=&FN_FrontEndVector2Modifier(Front End:Vector2f=<60 190>, Gameplay:Vector2f=<80 220>)
     1=&Multiply_Vector2D_ByFloat(Float=$System.FrontEndScale, Vector2D=&UniformRangedVector2D(Minimum=<20 40>, Maximum=<25 70>))
     2=&Multiply_Vector2D_ByFloat(
      Float=&Multiply_Float(
       A=&FN_FrontEndFloatModifier(Gameplay=1, Front End=0.45)
       B=&RandomRangeFloat(Minimum=4.5, Maximum=5)
      )
      Vector2D=<1 1.5>
     )
     3=&Multiply_Vector2D_ByFloat(
      Float=&FN_FrontEndFloatModifier(Gameplay=1, Front End=0.45)
      Vector2D=&RandomRangeVector2D(Minimum=<1.9 3.1487>, Maximum=<3 9.8129>)
     )
     4=&Multiply_Vector2D_ByFloat(Float=$System.FrontEndScale, Vector2D=&RandomRangeVector2D(Minimum=<100 170>, Maximum=<120 220>))
     Index=$Emitter.SpawnGroup
     Num=::NewEnumerator3
    )
    Sprite Size Mode=::NewEnumerator4
    Sprite UV Mode=::NewEnumerator1
    UsePositionOffset=false
    Write Material Random=true
   )
   &ShapeLocation(
    Cylinder Height=&MergeHelper_FloatSwitch(0=0, 1=0, 2=80, 3=?, Index=$Emitter.SpawnGroup, Num=::NewEnumerator1)
    Cylinder Radius=&MergeHelper_FloatSwitch(0=0, 1=0, 2=6, 3=?, Index=$Emitter.SpawnGroup, Num=::NewEnumerator1)
    MaskSpawnGroup=false
    Non Uniform Scale=&VectorFromFloat(Value=$System.FrontEndScale)
    Override Seed=false
    Rotation Axis=$System.DirectionVector
    Rotation Mode=::NewEnumerator1
    Shape Primitive=::NewEnumerator1
    Surface Only=false
    UseOwnerScale=true
   )
   &SubUVAnimation(
    End Frame Range Override=3
    Evaluation Type=::NewEnumerator0
    "Sprite Renderer "=DataInterface() {propertyValues {SpriteRenderer=&self:Directional_Elements_0.NiagaraSpriteRendererProperties_0}}
    SubUV Animation Mode:ENiagara_SubUVLookupModeV2=::NewEnumerator2
    UseEndFrame=true
    UseStartFrame=true
   )
   &AddVelocity(
    Module.Use Falloff Distance=false
    Override Seed=false
    Velocity Mode=::NewEnumerator1
    Velocity Speed=&MergeHelper_FloatSwitch(
     0=0
     1=&Multiply_Float(A=20, B=$System.FrontEndScale)
     2=&Multiply_Float(A=&RandomRangeFloat(Minimum=225, Maximum=300), B=$System.FrontEndScale)
     3=&Multiply_Float(A=&RandomRangeFloat(Minimum=200, Maximum=400), B=$System.FrontEndScale)
     4=0
     Index=$Emitter.SpawnGroup
     Num=::NewEnumerator3
    )
   ) {module=AddVelocity001}
   &AddVelocity(
    Rotation Coordinate Space=::Simulation
    Velocity=$System.DirectionVector
    Velocity Speed Scale=&Multiply_Float(
     A=&MergeHelper_FloatSwitch(
      0=&RandomRangeFloat(Minimum=180, Maximum=220)
      1=100
      2=&RandomRangeFloat(Minimum=300, Maximum=700)
      3=0
      4=150
      Index=$Emitter.SpawnGroup
      Num=::NewEnumerator3
     )
     B=$System.FrontEndScale
    )
   )
   &AddVelocity(Velocity=&MergeHelper_VectorSwitch(
     0=<0 0 0>
     1=<0 0 0>
     2=<0 0 0>
     3=&Multiply_Vector(
      A=&RandomRangeVector(Minimum=<100 0 0>, Maximum=<200 0 0>)
      B=&VectorFromFloat(Value=$System.FrontEndScale)
     )
     4=<0 0 0>
     Index=$Emitter.SpawnGroup
     Num=::NewEnumerator3
    )) {module=AddVelocity002}
   &DynamicMaterialParameters(
    Index 0 Param 1=0.2
    Index 0 Param 2=1
    Index 3 Param 1=1.5
    Index 3 Param 2=1
    Module.Index0Float0Write=true
    Module.Index0Float1Write=true
    Module.Index0Float2Write=true
    Module.Index0Float3Write=true
    Module.Index3Float0Write=true
    Module.Index3Float1Write=true
    Module.Index3Float2Write=true
    Module.Index3Float3Write=true
    Write Parameter Index 3=::NewEnumerator1
   )
   /SetVariables(Particles.DynamicParameterAlt:Vector4f=&MakeVector4(X=1, Y=1, Z=1, W=1)) {isSetParameters=true}
   &CameraOffset(Camera Offset Amount=&MergeHelper_FloatSwitch(0=30, 1=25, 2=20, 3=30, Index=$Emitter.SpawnGroup, Num=::NewEnumerator2))
   /SetVariables() {isSetParameters=true}
   /SetVariables(Particles.VisibilityTag=$Emitter.SpawnGroup) {isSetParameters=true}
  ]
  particleUpdateScript [
   &ParticleState()
   &AccelerationForce(Acceleration=&MergeHelper_VectorSwitch(0=<0 0 0>, 1=<0 0 0>, 2=<0 0 -500>, 3=<0 0 -160>, 4=<0 0 0>, Index=$Particles.VisibilityTag, Num=::NewEnumerator3))
   &Drag(
    Drag=&MergeHelper_FloatSwitch(
     0=0
     1=0
     2=&RandomRangeFloat(Minimum=5, Maximum=8)
     3=&FloatFromCurve(FloatCurve=DataInterface() {propertyValues {Curve {keys [<at 0 1>,<at 1 6>]}}}, Scale Curve=1.2)
     4=3
     Index=$Particles.VisibilityTag
     Num=::NewEnumerator3
    )
    Rotational Drag:NiagaraFloat=&MergeHelper_FloatSwitch(0=0, 1=0, 2=1, 3=1, 4=1, Index=$Particles.VisibilityTag, Num=::NewEnumerator3)
    Use Linear Drag=true
    Use Rotational Drag=true
   )
   &CurlNoiseForce(
    Bias Noise=false
    Noise Strength=&MergeHelper_FloatSwitch(
     0=0
     1=0
     2=0
     3=&RandomRangeFloat(Minimum=200, Maximum=400)
     4=0
     Index=$Particles.VisibilityTag
     Num=::NewEnumerator3
    )
    Pan Noise=false
   )
   &SolveForcesAndVelocity(Clamp Velocity=false, Limit Acceleration=false)
   &ScaleSpriteSize(
    Scale Factor=&MergeHelper_Vector2Switch(
     0=&Vector2DFromCurve(
      Vector2Curve=DataInterface() {propertyValues {
        XCurve {keys [{tangentMode=RCTM_User, time=0, value=0}, <at 0.2 1.075>, {tangentMode=RCTM_User, time=1, value=-0.1}]}
        YCurve {keys [{tangentMode=RCTM_User, time=0, value=0}, <at 0.1 1.025>, {tangentMode=RCTM_User, time=1, value=0.8}]}
       }}
      ScaleCurve=<1 1>
     )
     1=&Vector2DFromCurve(
      Vector2Curve=DataInterface() {propertyValues {
        XCurve {keys [<at 0 1 RCIM_Cubic>,<at 1 0 RCIM_Cubic>]}
        YCurve {keys [{arriveTangent=3.9291, interpMode=RCIM_Cubic, leaveTangent=3.9291, tangentMode=RCTM_User, time=0, value=0}, <at 1 2 RCIM_Cubic>]}
       }}
      ScaleCurve=<1.5 1>
     )
     2=&Vector2DFromFloat(Value=&FloatFromCurve(
       FloatCurve=DataInterface() {propertyValues {Curve {keys [{arriveTangent=-6.725, interpMode=RCIM_Cubic, leaveTangent=-6.725, tangentMode=RCTM_User, time=0.65, value=1}, <at 1 0 RCIM_Cubic>]}}}
       Scale Curve=1.5
      ))
     3=&Vector2DFromFloat(Value=&FloatFromCurve(FloatCurve=DataInterface() {propertyValues {Curve {keys [
           <at 0 0 RCIM_Cubic>
           {interpMode=RCIM_Cubic, tangentMode=RCTM_User, time=0.118, value=1}
           {interpMode=RCIM_Cubic, tangentMode=RCTM_User, time=1, value=0}
          ]}}}))
     4=<1 1>
     Index=$Particles.VisibilityTag
     Num=::NewEnumerator3
    )
    Scale Sprite Size Mode=::NewEnumerator2
   )
   &ScaleSpriteSizeBySpeed(
    Max Scale Factor=&MergeHelper_Vector2Switch(0=<1 1>, 1=<1 1>, 2=<0.2 5>, 3=<1 6>, 4=<1 1>, Index=$Particles.VisibilityTag, Num=::NewEnumerator3)
    Min Scale Factor=&MergeHelper_Vector2Switch(0=<1 1>, 1=<1 1>, 2=<0.2 0.2>, 3=<1 3>, 4=<1 1>, Index=$Particles.VisibilityTag, Num=::NewEnumerator3)
   )
   /SetVariables(Particles.DynamicParameterAlt1:Vector4f=&MakeVector4(
     X=&FloatFromCurve(FloatCurve=DataInterface() {propertyValues {Curve {keys [<at 0 0>,<at 1 1>]}}})
     Y=&UniformAOrBFloat(A=0, Evaluation Type=::NewEnumerator0, Recalculate Random Each Loop=true, Override Seed=false)
     Z=0
     W=0
    )) {isSetParameters=true}
   &ScaleColor(
    Scale Alpha=&MergeHelper_FloatSwitch(
     0=$System.ToDScalar
     1=&FloatFromCurve(FloatCurve=DataInterface() {propertyValues {Curve {keys [<at 0 1 RCIM_Cubic>,<at 1 0>]}}}, Scale Curve=$System.ToDScalar)
     2=$System.ToDScalar
     3=&FloatFromCurve(
      FloatCurve=DataInterface() {propertyValues {Curve {keys [
          {arriveTangent=-0.0852, interpMode=RCIM_Cubic, leaveTangent=-0.0852, tangentMode=RCTM_User, time=0, value=1}
          {arriveTangent=-2.6621, interpMode=RCIM_Cubic, leaveTangent=-2.6621, tangentMode=RCTM_User, time=1, value=0}
         ]}}}
      Scale Curve=$System.ToDScalar
     )
     4=&FloatFromCurve(
      FloatCurve=DataInterface() {propertyValues {Curve {keys [<at 0 1>,<at 0.8091 0.9932>,<at 1 0>]}}}
      Scale Curve=&FNTimeOfDayScalar(Night Scale=5, Morning Scale=4, Day Scale=2, Evening Scale=4)
     )
     Index=$Particles.VisibilityTag
     Num=::NewEnumerator3
    )
    Scale RGB=&MergeHelper_VectorSwitch(
     0=<1 1 1>
     1=<1 1 1>
     2=&VectorFromFloat(Value=&FloatFromCurve(
       FloatCurve=DataInterface() {propertyValues {Curve {keys [
           {interpMode=RCIM_Cubic, tangentMode=RCTM_User, time=0, value=1}
           <at 0.3 0.2>
           {interpMode=RCIM_Cubic, tangentMode=RCTM_User, time=1, value=0}
          ]}}}
       Scale Curve=10
      ))
     3=&VectorFromFloat(Value=&FloatFromCurve(FloatCurve=DataInterface() {propertyValues {Curve {keys [
           {interpMode=RCIM_Cubic, tangentMode=RCTM_User, time=0, value=1}
           {arriveTangent=-1.4115, interpMode=RCIM_Cubic, leaveTangent=-1.4115, tangentMode=RCTM_User, time=1, value=0}
          ]}}}))
     4=&MakeVectorFromLinearColorRGB(Color=&FNTimeOfDayColor(
       Night Color=<rgba 0.34 0.51 0.85 1>
       Morning Color=<rgba 0.902 0.5882 0.5412 1>
       Day Color=<rgba 1 1 1 1>
       Evening Color=<rgba 0.9 0.54 0.6 1>
      ))
     Index=$Particles.VisibilityTag
     Num=::NewEnumerator3
    )
    ScaleA=true
    ScaleRGB=true
   )
  ]
  rendererClasses [&NiagaraSpriteRendererProperties]
  renderers [
   {rendererClass=&NiagaraSpriteRendererProperties}
   {rendererClass=&NiagaraSpriteRendererProperties, rendererIndex=1}
   {rendererClass=&NiagaraSpriteRendererProperties, rendererIndex=2}
   {rendererClass=&NiagaraSpriteRendererProperties, rendererIndex=3}
   {rendererClass=&NiagaraSpriteRendererProperties, rendererIndex=4}
  ]
 }
 Fire_Sprites {
  bEnabled=false
  emitterUpdateScript [
   &EmitterState(Life Cycle Mode=::NewEnumerator1, Loop Behavior=::NewEnumerator1, Loop Delay=0.025, Loop Duration=0.125, UseLoopDelay=true)
   /Game/Effects/Niagara/Modules/Spawn/FN_Spawn_TriggeredBurst(Cooldown=0.005, Module.Use Burst Count Limit=false, Spawn Count=3, UseLoopCountLimit=false)
   &SpawnRate(SpawnRate=50, Use Spawn Probability=false)
  ]
  particleSpawnScript [
   &InitializeParticle(
    Color=&RandomRangeLinearColor(Minimum=<rgba 1 1 1 1>, Maximum=<rgba 2 2 2 1>)
    Lifetime=&RandomRangeFloat(Minimum=0.25, Maximum=0.75)
    Position Offset=&Multiply_VectorByFloat(Vector=$System.DirectionVector, Float=300)
    Position Offset Coordinate Space=::Local
    Sprite Size Mode=::NewEnumerator1
    Uniform Sprite Size Max=15
    Uniform Sprite Size Min=10
    UsePositionOffset=true
    Write Material Random=true
   )
   /Game/Effects/Niagara/Modules/Curves/FN_ListCurve(
    Apply Position=true
    Curve Index:NiagaraFloat=&RandomRangeFloat(Minimum=0, Maximum=1)
    Input Type=2
    Number Inputs Switch=0
    Position Cood Interpolation:NiagaraFloat=&FloatFromCurve(FloatCurve=DataInterface() {propertyValues {Curve {keys [<at 0 0>,<at 1 1>]}}})
    Position Coordinate Space=2
   ) {enabled=false}
   /Game/Effects/Niagara/Modules/Curves/FN_ListCurve_SkeletalMeshSampler(
    Apply Position=true
    Curve Index:NiagaraFloat=&RandomRangeFloat(Minimum=0, Maximum=1)
    Index_02=1
    Number of Inputs=0
    Skeletal Mesh:NiagaraDataInterfaceSkeletalMesh=$System.SkeletalMesh
   ) {enabled=false, module=FN_ListCurve_SkeletalMeshSampler001}
   &AddVelocity(Rotation Coordinate Space=::World, Velocity=$System.DirectionVector, Velocity Speed Scale=100) {enabled=false}
  ]
  particleUpdateScript [
   &ParticleState()
   &SpriteRotationRate(Rotation Rate=0) {enabled=false}
   &ScaleSpriteSize(
    Non-Uniform Curve Scale=&RandomRangeVector2D(Minimum=<1 1>, Maximum=<2 3>)
    Non-Uniform Curve Sprite Scale=DataInterface() {propertyValues {
      XCurve {keys [{tangentMode=RCTM_User, time=0.0066, value=1}, {tangentMode=RCTM_User, time=1, value=0.5}]}
      YCurve {keys [{tangentMode=RCTM_User, time=0.0132, value=0}, {tangentMode=RCTM_User, time=1.0066, value=2}]}
     }}
    Scale Sprite Size Mode=::NewEnumerator3
   )
   &DynamicMaterialParameters(
    Index 0 Param 1=&FloatFromCurve(FloatCurve=DataInterface() {propertyValues {Curve {keys [{interpMode=RCIM_Cubic, tangentMode=RCTM_User, time=0.2, value=0}, <at 1 1>, <at 1 1>]}}})
    Index 0 Param 2=1
    Index 0 Param 3=&FloatFromCurve(FloatCurve=DataInterface() {propertyValues {Curve {keys [
         {arriveTangent=-0.1755, interpMode=RCIM_Cubic, leaveTangent=-0.1755, tangentMode=RCTM_User, time=0, value=1}
         {arriveTangent=-1.7001, interpMode=RCIM_Cubic, tangentMode=RCTM_User, time=0.1, value=-0.013}
         {arriveTangent=1.5083, tangentMode=RCTM_User, time=1, value=1}
        ]}}})
    Index 0 Param 4=1
    Index 1 Param 1=&RandomRangeFloat(Minimum=0, Maximum=1)
    Index 1 Param 2=&RandomRangeFloat(Minimum=0, Maximum=1)
    Index 2 Param 1=&RandomRangeFloat(Minimum=0, Maximum=1)
    Index 2 Param 2=&RandomRangeFloat(Minimum=0, Maximum=1)
    Index 2 Param 3=&RandomRangeFloat(Minimum=0.5, Maximum=2)
    Index 2 Param 4=&FloatFromCurve(
     FloatCurve=DataInterface() {propertyValues {Curve {keys [{interpMode=RCIM_Cubic, tangentMode=RCTM_User, time=0, value=1}, {arriveTangent=-0.9865, tangentMode=RCTM_User, time=1, value=0.5}]}}}
     Scale Curve=&RandomRangeFloat(Minimum=0.75, Maximum=1)
    )
    Module.Index0Float0Write=true
    Module.Index0Float1Write=true
    Module.Index0Float2Write=true
    Module.Index0Float3Write=true
    Module.Index1Float0Write=true
    Module.Index1Float1Write=true
    Module.Index1Float2Write=true
    Module.Index1Float3Write=true
    Module.Index2Float0Write=true
    Module.Index2Float1Write=true
    Module.Index2Float2Write=true
    Module.Index2Float3Write=true
    Write Parameter Index 1=::NewEnumerator1
    Write Parameter Index 2=::NewEnumerator1
   )
   &ScaleColor(
    Scale Alpha=&FloatFromCurve(FloatCurve=DataInterface() {propertyValues {Curve {keys [<at 0 0.8>,<at 1 0.8>]}}})
    ScaleA=false
    ScaleRGB=true
   )
   &CameraOffset(Camera Offset Amount=-3)
   &FN_ListCurve(
    Apply Position=true
    Input Type=2
    Number Inputs Switch=0
    Position Cood Interpolation:NiagaraFloat=&FloatFromCurve(FloatCurve=DataInterface() {propertyValues {Curve {keys [<at 0.5 0>,<at 1 1>]}}})
    Position Coordinate Space=2
   ) {enabled=false, module=FN_ListCurve001}
   &FN_ListCurve(
    Apply Position=true
    Custom Key Interpolation=true
    Input Type=2
    Key to Key Interpolation=DataInterface() {propertyValues {Curve {keys [
        {arriveTangent=-0.8981, interpMode=RCIM_Cubic, leaveTangent=-0.8981, tangentMode=RCTM_User, time=0, value=0}
        {arriveTangent=4.7773, tangentMode=RCTM_User, time=1, value=1}
       ]}}}
    Number Inputs Switch=0
    Position Cood Interpolation:NiagaraFloat=&FloatFromCurve(FloatCurve=DataInterface() {propertyValues {Curve {keys [<at 0.25 0>,<at 0.75 1>]}}})
    Position Coordinate Space=2
    Vector_01:Vector3f=&RandomRangeVector(Minimum=<0 0 60>, Maximum=<0 0 70>)
    Vector_02:Vector3f=&RandomRangeVector(Minimum=<-5 -20 0>, Maximum=<5 -10 0>)
   ) {enabled=false, module=FN_ListCurve002}
   &FN_ListCurve_SkeletalMeshSampler(
    Apply Orientation=true
    Apply Position=true
    Bone Coordinate Space=1
    Bone Position Contribution:NiagaraFloat=&FloatFromCurve(FloatCurve=DataInterface() {propertyValues {Curve {keys [{interpMode=RCIM_Cubic, tangentMode=RCTM_User, time=0.5, value=1}, {arriveTangent=-12.0126, tangentMode=RCTM_User, time=0.8, value=0}]}}})
    Curve Index:NiagaraFloat=&FloatFromCurve(FloatCurve=DataInterface() {propertyValues {Curve {keys [{interpMode=RCIM_Cubic, tangentMode=RCTM_User, time=0, value=0.2}, <at 1 1>]}}})
    Index_02=1
    Index_03=1
    Number of Inputs=1
    Offset Coordinate Space=3
    Offset_01:Vector3f=&RandomRangeVector(Minimum=<0 0 0>, Maximum=<0 0 0>)
    Offset_02:Vector3f=&RandomRangeVector(Minimum=<-4 -2 -20>, Maximum=<-4 2 -15>)
    Offset_03:Vector3f=&RandomRangeVector(Minimum=<5 -10 -15>, Maximum=<15 10 -10>)
    Selection Type=0
    Skeletal Mesh:NiagaraDataInterfaceSkeletalMesh=$System.SkeletalMesh
   ) {enabled=false}
   &SpriteFacingAndAlignment(
    Alignment Coordinate Space=::World
    Sprite Alignment=&Lerp_Vector(
     A=&Multiply_VectorByFloat(
      Vector=&ReturnMeshOrientationAxis(Mesh Orientation Axis=::XAxis, Mesh Orientation Quat=$Particles.MeshOrientation)
      Float=&RandomRangeFloat(Minimum=-1, Maximum=1)
     )
     B=<0 1 1>
     Alpha=&FloatFromCurve(FloatCurve=DataInterface() {propertyValues {Curve {keys [
          {interpMode=RCIM_Cubic, tangentMode=RCTM_User, time=0, value=0.5}
          {arriveTangent=-0.625, interpMode=RCIM_Cubic, leaveTangent=-0.625, time=0.45, value=0.9794}
          <at 0.8 0>
         ]}}})
    )
    Sprite Facing=<0 0 1>
    WriteAlignment=true
    WriteFacing=false
   ) {enabled=false}
   &SolveForcesAndVelocity(Clamp Velocity=false, Limit Acceleration=false)
  ]
  rendererClasses [&NiagaraSpriteRendererProperties, &NiagaraMeshRendererProperties]
  renderers [
   {rendererClass=&NiagaraSpriteRendererProperties}
   {rendererClass=&NiagaraMeshRendererProperties, rendererIndex=1}
  ]
 }
 Minimal {
  emitterUpdateScript [
   &EmitterState()
   &SpawnBurst_Instantaneous(Use Spawn Probability=false, UseLoopCountLimit=false)
  ]
  particleSpawnScript [
   &InitializeParticle(Lifetime=0.75, Sprite Size=<35 125>, Sprite Size Mode=::NewEnumerator4, UsePositionOffset=false, Write Material Random=true)
   &AddVelocity(Velocity=$System.DirectionVector, Velocity Speed Scale=300)
  ]
  particleUpdateScript [
   &ParticleState()
   &ScaleSpriteSize(
    Non-Uniform Curve Sprite Scale=DataInterface() {propertyValues {
      XCurve {keys [{interpMode=RCIM_Cubic, tangentMode=RCTM_User, time=0, value=1}, {arriveTangent=-0.811, tangentMode=RCTM_User, time=1, value=0.5}]}
      YCurve {keys [{interpMode=RCIM_Cubic, tangentMode=RCTM_User, time=0, value=1}, {arriveTangent=2.0274, tangentMode=RCTM_User, time=0.25, value=2}]}
     }}
    Scale Sprite Size Mode=::NewEnumerator3
   )
   &DynamicMaterialParameters(
    Index 0 Param 1=1
    Index 0 Param 2=1
    Index 0 Param 3=&FloatFromCurve(FloatCurve=DataInterface() {propertyValues {Curve {keys [{interpMode=RCIM_Cubic, tangentMode=RCTM_User, time=0, value=0}, <at 1 1>]}}})
    Index 0 Param 4=1
    Index 1 Param 1=&RandomRangeFloat(Minimum=0, Maximum=1)
    Index 1 Param 2=&RandomRangeFloat(Minimum=0, Maximum=1)
    Module.Index0Float0Write=true
    Module.Index0Float1Write=true
    Module.Index0Float2Write=true
    Module.Index0Float3Write=true
    Module.Index1Float0Write=true
    Module.Index1Float1Write=true
    Module.Index1Float2Write=true
    Module.Index1Float3Write=true
    Write Parameter Index 1=::NewEnumerator1
   )
   &SolveForcesAndVelocity(Clamp Velocity=false, Limit Acceleration=false)
   &CameraOffset(Camera Offset Amount=5)
   /SetVariables(
    Particles.Radial Fire Dyn1:Vector4f=&MakeVector4(
     X=1
     Y=1
     Z=&FloatFromCurve(FloatCurve=DataInterface() {propertyValues {Curve {keys [{interpMode=RCIM_Cubic, tangentMode=RCTM_User, time=0.25, value=0}, <at 0.75 1>]}}})
     W=1
    )
    Particles.Radial Fire Pos:NiagaraPosition=&AddVectorToPosition(
     Position=$Engine.Emitter.SimulationPosition
     Vector=&Multiply_VectorByFloat(
      Vector=<0 0 1>
      Float=&FloatFromCurve(FloatCurve=DataInterface() {propertyValues {Curve {keys [<at 0 0>,<at 1 1>]}}}, Scale Curve=75)
     )
    )
    Particles.Radial Fire Size:Vector2f=&Multiply_Vector2D_ByFloat(
     Float=&FloatFromCurve(FloatCurve=DataInterface() {propertyValues {Curve {keys [<at 0 1>,<at 1 4>]}}}, Scale Curve=25)
     Vector2D=<1 1.5>
    )
    Particles.Radial Fire Velocity:Vector3f=&Multiply_VectorByFloat(Vector=<0 0 1>, Float=300)
   ) {isSetParameters=true}
  ]
  rendererClasses [&NiagaraSpriteRendererProperties, &NiagaraMeshRendererProperties]
  renderers [
   {rendererClass=&NiagaraSpriteRendererProperties}
   {rendererClass=&NiagaraMeshRendererProperties, rendererIndex=1}
   {rendererClass=&NiagaraSpriteRendererProperties, rendererIndex=2}
  ]
 }
}
summary {
 systemName=NS_Pickaxe_DuneBriefMerry_Impact
 emitters [
  {emitterName=Directional_Elements, rendererClasses [&NiagaraSpriteRendererProperties]}
  {emitterName=Minimal, rendererClasses [&NiagaraSpriteRendererProperties, &NiagaraMeshRendererProperties]}
  {
   bEnabled=false
   emitterName=Fire_Sprites
   rendererClasses [&NiagaraSpriteRendererProperties, &NiagaraMeshRendererProperties]
  }
 ]
 userVariables [
  User.bIsFrontEnd {type=NiagaraBool}
  User.bIsFrontEndPreview {type=NiagaraBool}
  User.PhysicalSurface {type=NiagaraInt32}
  User.SwingSectionIndex {type=NiagaraInt32}
  User.VariantColor {type=NiagaraFloat, defaultValue=0, description=""}
  User.VariantToon {type=NiagaraFloat, defaultValue=0, description=""}
 ]
}
