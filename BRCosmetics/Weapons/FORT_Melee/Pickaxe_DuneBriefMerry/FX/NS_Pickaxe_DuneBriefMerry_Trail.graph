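# Niagara swing trail effect for the DuneBriefMerry pickaxe cosmetic, combining an animation trail, a fire afterimage and rising sprites.
# Trail uses anim-trail modules with two ribbon renderers, while Idle_Fire_AfterImage bursts three partitioned particles into a ribbon positioned by skeletal mesh sampling.
# Fire_Sprites emits short-lived, upward-drifting sprites with randomized dynamic material parameters.
# User.SwingSectionIndex selects per-swing trail color and material curve, plus afterimage lifetime and mirrored offsets.
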NiagaraSystem /BRCosmetics/Weapons/FORT_Melee/Pickaxe_DuneBriefMerry/FX/NS_Pickaxe_DuneBriefMerry_Trail
dependencies {
 usedDataInterfaces [
  &NiagaraDataInterfaceArrayColor
  &NiagaraDataInterfaceCurve
  &NiagaraDataInterfaceArrayFloat
  &NiagaraDataInterfaceArrayInt32
  &NiagaraDataInterfaceArrayFloat2
  &NiagaraDataInterfaceVector2DCurve
  &NiagaraDataInterfaceSkeletalMesh
  &NiagaraDataInterfaceColorCurve
 ]
 usedDynamicInputs [&SetBoolByFloatComparison]
 usedModules [
  &self:SystemSpawnScript.SystemScriptSource.SystemScriptGraph.NiagaraNodeAssignment_0.SetVariables_#1
  &SystemState
  &self:SystemSpawnScript.SystemScriptSource.SystemScriptGraph.NiagaraNodeAssignment_1.SetVariables_#2
  &self:SystemSpawnScript.SystemScriptSource.SystemScriptGraph.NiagaraNodeAssignment_2.SetVariables_#3
 ]
}
emitters {
 Fire_Sprites {
  emitterUpdateScript [
   &EmitterState(Life Cycle Mode=::NewEnumerator1, Loop Behavior=::NewEnumerator1, Loop Delay=0.025, Loop Duration=0.125, UseLoopDelay=true)
   /Game/Effects/Niagara/Modules/Spawn/FN_Spawn_TriggeredBurst(Cooldown=0.0025, Module.Use Burst Count Limit=false, Spawn Count=3, UseLoopCountLimit=false)
   &SpawnRate(SpawnRate=75, Use Spawn Probability=false)
  ]
  particleSpawnScript [
   &InitializeParticle(
    Color=&RandomRangeLinearColor(Minimum=<rgba 1 1 1 1>, Maximum=<rgba 2 2 2 1>)
    Lifetime=&RandomRangeFloat(Minimum=0.35, Maximum=0.75)
    Position Offset=&RandomRangeVector(Minimum=<-25 0 40>, Maximum=<25 0 50>)
    Position Offset Coordinate Space=::Local
    Sprite Size Mode=::NewEnumerator1
    Uniform Sprite Size Max=12
    Uniform Sprite Size Min=8
    UsePositionOffset=true
    Write Material Random=true
   )
   &AddVelocity(
    Rotation Coordinate Space=::World
    Velocity=&RandomRangeVector(Minimum=<-0.2 -0.2 0.75>, Maximum=<0.2 0.2 1>)
    Velocity Speed Scale=50
   )
  ]
  particleUpdateScript [
   &ParticleState()
   &ScaleSpriteSize(
    Non-Uniform Curve Scale=&RandomRangeVector2D(Minimum=<1 1>, Maximum=<1 3>)
    Non-Uniform Curve Sprite Scale=DataInterface() {propertyValues {
      XCurve {keys [{tangentMode=RCTM_User, time=0, value=1}, {tangentMode=RCTM_User, time=1, value=0.5}]}
      YCurve {keys [{tangentMode=RCTM_User, time=0, value=0.5}, {tangentMode=RCTM_User, time=1.0066, value=2}]}
     }}
    Scale Sprite Size Mode=::NewEnumerator3
   )
   &DynamicMaterialParameters(
    Index 0 Param 1=&FloatFromCurve(FloatCurve=DataInterface() {propertyValues {Curve {keys [{interpMode=RCIM_Cubic, tangentMode=RCTM_User, time=0.2, value=0}, <at 0.5 1>, <at 1 1>]}}})
    Index 0 Param 2=1
    Index 0 Param 3=&FloatFromCurve(FloatCurve=DataInterface() {propertyValues {Curve {keys [
         {arriveTangent=-0.1755, interpMode=RCIM_Cubic, leaveTangent=-0.1755, tangentMode=RCTM_User, time=0, value=1}
         {arriveTangent=-1.7001, interpMode=RCIM_Cubic, tangentMode=RCTM_User, time=0.1, value=-0.013}
         {arriveTangent=1.5083, tangentMode=RCTM_User, time=1, value=1}
        ]}}})
    Index 0 Param 4=1
    Index 1 Param 1=&RandomRangeFloat(Minimum=0, Maximum=1)
    Index 1 Param 2=&RandomRangeFloat(Minimum=0, Maximum=1)
    Index 2 Param 1=&RandomRangeFloat(Minimum=0, Maximum=1)
    Index 2 Param 2=&RandomRangeFloat(Minimum=0, Maximum=1)
    Index 2 Param 3=&RandomRangeFloat(Minimum=0.5, Maximum=2)
    Index 2 Param 4=&FloatFromCurve(
     FloatCurve=DataInterface() {propertyValues {Curve {keys [{interpMode=RCIM_Cubic, tangentMode=RCTM_User, time=0, value=1}, {arriveTangent=-0.9865, tangentMode=RCTM_User, time=1, value=0.5}]}}}
     Scale Curve=&RandomRangeFloat(Minimum=0.75, Maximum=1)
    )
    Module.Index0Float0Write=true
    Module.Index0Float1Write=true
    Module.Index0Float2Write=true
    Module.Index0Float3Write=true
    Module.Index1Float0Write=true
    Module.Index1Float1Write=true
    Module.Index1Float2Write=true
    Module.Index1Float3Write=true
    Module.Index2Float0Write=true
    Module.Index2Float1Write=true
    Module.Index2Float2Write=true
    Module.Index2Float3Write=true
    Write Parameter Index 1=::NewEnumerator1
    Write Parameter Index 2=::NewEnumerator1
   )
   &ScaleColor(
    Scale Alpha=&FloatFromCurve(FloatCurve=DataInterface() {propertyValues {Curve {keys [<at 0 0.8>,<at 1 0.8>]}}})
    ScaleA=false
    ScaleRGB=true
   )
   &CameraOffset(Camera Offset Amount=-3)
   &SolveForcesAndVelocity(Clamp Velocity=false, Limit Acceleration=false)
  ]
  rendererClasses [&NiagaraSpriteRendererProperties]
  renderers [{rendererClass=&NiagaraSpriteRendererProperties}]
 }
 Idle_Fire_AfterImage {
  emitterUpdateScript [
   &EmitterState()
   &SpawnBurst_Instantaneous(Spawn Count=3, Use Spawn Probability=false, UseLoopCountLimit=false)
  ]
  particleSpawnScript [
   &PartitionParticles(Normalized Index Includes One=true, Number Of Partitions=3)
   &InitializeParticle(
    Color=&SelectLinearColorFromArray(
     Array Sampling Mode=::NewEnumerator1
     Direct Array Index=$Particles.PartitionParticles.Partition
     Color Selection Array=DataInterface() {propertyValues {ColorData [<rgba 1 1 1 1>, <rgba 1 1 1 1>, <rgba 1 1 1 1>]}}
     Random Seed=?
    )
    Lifetime=&MergeHelper_FloatSwitch(0=0.4, 1=0.25, 2=0.4, 3=0.3, Index=$User.SwingSectionIndex, Num=::NewEnumerator2)
    Mesh Scale=<0.8 0.8 0.9>
    Mesh Scale Mode=::NewEnumerator4
    Position Offset=<0 0 18>
    Ribbon Width=&SelectFloatFromArray(Direct Array Index=$Particles.PartitionParticles.Partition, Float Selection Array=DataInterface() {propertyValues {FloatData [10,35,1]}})
    Ribbon Width Mode=::NewEnumerator1
    Sprite Size Mode=::NewEnumerator3
    UsePositionOffset=false
    Write Material Random=true
   )
  ]
  particleUpdateScript [
   &ParticleState()
   &DynamicMaterialParameters(
    Index 0 Param 1=1
    Index 0 Param 2=1
    Index 0 Param 3=&FloatFromCurve(FloatCurve=DataInterface() {propertyValues {Curve {keys [
         {interpMode=RCIM_Cubic, tangentMode=RCTM_User, time=0, value=1}
         {interpMode=RCIM_Cubic, tangentMode=RCTM_User, time=0.05, value=0.2}
         {arriveTangent=-0.089, interpMode=RCIM_Cubic, leaveTangent=-0.089, tangentMode=RCTM_User, time=0.9, value=0.2}
         <at 1 1>
        ]}}})
    Index 0 Param 4=1
    Index 2 Param 1=&Waveform("Waveform [1]"=::NewEnumerator0, "[1] Frequency"=1, "[1] Amplitude Min/Max"=<0 0.05>)
    Index 2 Param 2=&Waveform("Waveform [1]"=::NewEnumerator7, "[1] Frequency"=3, "[1] Amplitude Min/Max"=<0 0.05>)
    Module.Index0Float0Write=true
    Module.Index0Float1Write=true
    Module.Index0Float2Write=true
    Module.Index0Float3Write=true
    Module.Index2Float0Write=true
    Module.Index2Float1Write=true
    Module.Index2Float2Write=true
    Module.Index2Float3Write=true
    Write Parameter Index 2=::NewEnumerator1
   )
   /Game/Effects/Niagara/Modules/Curves/FN_ListCurve_SkeletalMeshSampler(
    Apply Orientation=true
    Apply Position=true
    Bone Coordinate Space=2
    Bone Position Contribution:NiagaraFloat=&SelectFloatFromArray(Direct Array Index=$Particles.PartitionParticles.Partition, Float Selection Array=DataInterface() {propertyValues {FloatData [1,0.5,0.6]}})
    Curve Index:NiagaraFloat=&Multiply_Float(
     A=&SelectFloatFromArray(Direct Array Index=$Particles.PartitionParticles.Partition, Float Selection Array=DataInterface() {propertyValues {FloatData [0,0.5,1]}})
     B=1
    )
    Index_02=1
    Index_03=1
    Number of Inputs=1
    Offset Coordinate Space=3
    Offset_01:Vector3f=<0 0 -1>
    Offset_02:Vector3f=&MergeHelper_VectorSwitch(0=<-20 0 0>, 1=<20 0 0>, 2=<-20 0 0>, 3=<20 0 0>, Index=$User.SwingSectionIndex, Num=::NewEnumerator2)
    Offset_03:Vector3f=&MergeHelper_VectorSwitch(0=<-35 0 20>, 1=<35 0 20>, 2=<-35 0 20>, 3=<35 0 20>, Index=$User.SwingSectionIndex, Num=::NewEnumerator2)
    Selection Type=0
    Skeletal Mesh:NiagaraDataInterfaceSkeletalMesh=$System.SkeletalMesh
   )
  ]
  rendererClasses [&NiagaraRibbonRendererProperties]
  renderers [{rendererClass=&NiagaraRibbonRendererProperties}]
 }
 Trail {
  emitterUpdateScript [
   &EmitterState()
   &FN_SpawnAnimTrail(Switch on ENiagara_AnimTrail_Source:ENiagara_AnimTrail_Source=::NewEnumerator1)
  ]
  particleSpawnScript [
   &InitializeParticle(Color=<rgba 0.75 0.75 0.75 1>, Lifetime=5, Ribbon Width Mode=::NewEnumerator1, UsePositionOffset=false, Write Material Random=true)
   &FN_InitializeAnimTrailBoneIndecies()
   &FN_InitializeAnimTrail(
    Color=&SelectLinearColorFromArray(
     Array Sampling Mode=::NewEnumerator1
     Direct Array Index=$User.SwingSectionIndex
     Color Selection Array=DataInterface() {propertyValues {ColorData [<rgba 1 1 1 1>, <rgba 0.75 0.75 0.75 1>, <rgba 1 1 1 1>, <rgba 0.75 0.75 0.75 1>]}}
     Random Seed=?
    )
    Lifetime=0.45
   )
   /SetVariables(Particles.cross width=20) {isSetParameters=true}
  ]
  particleUpdateScript [
   &ParticleState()
   &FN_UpdateAnimTrail()
   &DynamicMaterialParameters(
    Index 0 Param 1=1
    Index 0 Param 2=1
    Index 0 Param 3=&MergeHelper_FloatSwitch(
     0=&FloatFromCurve(FloatCurve=DataInterface() {propertyValues {Curve {keys [{interpMode=RCIM_Cubic, tangentMode=RCTM_User, time=0.8, value=0}, <at 1 1>]}}})
     1=&FloatFromCurve(FloatCurve=DataInterface() {propertyValues {Curve {keys [{interpMode=RCIM_Cubic, tangentMode=RCTM_User, time=0.85, value=0}, <at 1 1>]}}})
     2=&FloatFromCurve(FloatCurve=DataInterface() {propertyValues {Curve {keys [{interpMode=RCIM_Cubic, tangentMode=RCTM_User, time=0.9, value=0}, <at 1 1>]}}})
     3=&FloatFromCurve(FloatCurve=DataInterface() {propertyValues {Curve {keys [{interpMode=RCIM_Cubic, tangentMode=RCTM_User, time=0.8, value=0}, <at 1 1>]}}})
     Index=$User.SwingSectionIndex
     Num=::NewEnumerator2
    )
    Index 0 Param 4=1
    Index 1 Param 3=0.05
    Module.Index0Float0Write=true
    Module.Index0Float1Write=true
    Module.Index0Float2Write=true
    Module.Index0Float3Write=true
    Module.Index1Float0Write=true
    Module.Index1Float1Write=true
    Module.Index1Float2Write=true
    Module.Index1Float3Write=true
    Write Parameter Index 1=::NewEnumerator1
   )
   &ScaleColor(
    Scale Alpha=&FloatFromCurve(
     FloatCurve=DataInterface() {propertyValues {Curve {keys [{interpMode=RCIM_Cubic, tangentMode=RCTM_User, time=0, value=1}, <at 0.3 0>]}}}
     CurveIndex=$Particles.Age
    )
    ScaleA=true
    ScaleRGB=true
   )
  ]
  rendererClasses [&NiagaraRibbonRendererProperties]
  renderers [
   {rendererClass=&NiagaraRibbonRendererProperties}
   {rendererClass=&NiagaraRibbonRendererProperties, rendererIndex=1}
  ]
 }
}
summary {
 systemName=NS_Pickaxe_DuneBriefMerry_Trail
 emitters [
  {emitterName=Trail, rendererClasses [&NiagaraRibbonRendererProperties]}
  {emitterName=Idle_Fire_AfterImage, rendererClasses [&NiagaraRibbonRendererProperties]}
  {emitterName=Fire_Sprites, rendererClasses [&NiagaraSpriteRendererProperties]}
 ]
 userVariables [
  User.bIsFrontEnd {type=NiagaraBool}
  User.bIsFrontEndPreview {type=NiagaraBool}
  User.RibbonIDOffset {type=NiagaraInt32}
  User.SwingSectionIndex {type=NiagaraInt32}
  User.VariantColor {type=NiagaraFloat, defaultValue=0, description=""}
  User.VariantToon {type=NiagaraFloat, defaultValue=0, description=""}
 ]
}
